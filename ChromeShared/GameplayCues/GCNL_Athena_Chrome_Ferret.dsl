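; Looping gameplay cue that gives the Ferret plane vehicle Chrome-themed visual effects.
; On loop start it casts the target to the Fort vehicle interface and overrides tagged effects for prop grind, hit dust, player hit, large impact and boost with Chrome Niagara systems.
; It also swaps the propeller blur material for a Chrome version.
; The removal event is empty, so this cue does not itself restore the original effects.
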(event EventOnLoopingStart(EitherVFX-preferthisversion)
  (params (MyTarget :type Actor%20Object%20Reference) (Parameters :type Gameplay%20Cue%20Parameters%20Structure%20%28by%20ref%29) (ParticleComponents :type Array%20of%20FXSystem%20Component%20Object%20References) (AudioComponents :type Array%20of%20Audio%20Component%20Object%20References))
  (bind _asfort_vehicle_interface (Utilities|Casting|CastToFortVehicleInterface MyTarget))
  (Vehicle|FX|SetVehicleFx _asfort_vehicle_interface "(TagName=\"Athena.Vehicle.Ferret.FX.PropGrind\")" "/ChromeShared/FX/NS_Plane_Vehicle_Prop_Grind_Chrome.NS_Plane_Vehicle_Prop_Grind_Chrome")
  (Vehicle|FX|SetVehicleFx _asfort_vehicle_interface "(TagName=\"Athena.Vehicle.Ferret.FX.HitDust\")" "/ChromeShared/FX/NS_Plane_Vehicle_HitDust_Chrome.NS_Plane_Vehicle_HitDust_Chrome")
  (Vehicle|FX|SetVehicleFx _asfort_vehicle_interface "(TagName=\"Athena.Vehicle.Ferret.FX.HitDustPlayer\")" "/ChromeShared/FX/NS_Shopping_Cart_Vehicle_HitPlayer_Chrome.NS_Shopping_Cart_Vehicle_HitPlayer_Chrome")
  (Vehicle|FX|SetVehicleFx _asfort_vehicle_interface "(TagName=\"Athena.Vehicle.Ferret.FX.LargeImpact\")" "/ChromeShared/FX/NS_Impact_Large_Plane_Athena_Chrome.NS_Impact_Large_Plane_Athena_Chrome")
  (Vehicle|FX|SetVehicleFx _asfort_vehicle_interface "(TagName=\"Athena.Vehicle.Ferret.FX.Boost\")" "/ChromeShared/FX/NS_Plane_Boost_FX_Chrome.NS_Plane_Boost_FX_Chrome")
  (Vehicle|FX|SetVehicleFxMaterial _asfort_vehicle_interface "(TagName=\"Athena.Vehicle.Ferret.FX.PropBlur\")" "/ChromeShared/FX/Materials/M_Biplane_Propeller_Blur_Chrome.M_Biplane_Propeller_Blur_Chrome"))

(event EventOnRemoval(EitherVFX-preferthisversion)
  (params (MyTarget :type Actor%20Object%20Reference) (Parameters :type Gameplay%20Cue%20Parameters%20Structure%20%28by%20ref%29) (ParticleComponents :type Array%20of%20FXSystem%20Component%20Object%20References) (AudioComponents :type Array%20of%20Audio%20Component%20Object%20References) (BurstCameraShakeInstance :type Camera%20Shake%20Base%20Object%20Reference) (BurstDecalInstance :type Decal%20Actor%20Object%20Reference)))

(fn ConstructionScript)
