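; Looping gameplay cue notify that gives the Meatball vehicle its Chrome visual effects.
; On loop start, it casts the target to the Fort vehicle interface and overrides the tagged ground interaction, boost and large impact effects with Chrome Niagara systems.
; The removal event is empty, so this graph does not restore the previous effects.
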(event EventOnLoopingStart(EitherVFX-preferthisversion)
  (params (MyTarget :type Actor%20Object%20Reference) (Parameters :type Gameplay%20Cue%20Parameters%20Structure%20%28by%20ref%29) (ParticleComponents :type Array%20of%20FXSystem%20Component%20Object%20References) (AudioComponents :type Array%20of%20Audio%20Component%20Object%20References))
  (bind _asfort_vehicle_interface (Utilities|Casting|CastToFortVehicleInterface MyTarget))
  (Vehicle|FX|SetVehicleFx _asfort_vehicle_interface "(TagName=\"Athena.Vehicle.Meatball.FX.GroundInteraction\")" "/ChromeShared/FX/NS_Meatball_Chrome_GroundInteraction.NS_Meatball_Chrome_GroundInteraction")
  (Vehicle|FX|SetVehicleFx _asfort_vehicle_interface "(TagName=\"Athena.Vehicle.Meatball.FX.Boost\")" "/ChromeShared/FX/NS_Meatball_Chrome_Boost.NS_Meatball_Chrome_Boost")
  (Vehicle|FX|SetVehicleFx _asfort_vehicle_interface "(TagName=\"Athena.Vehicle.Meatball.FX.LargeImpact\")" "/ChromeShared/FX/NS_Impact_Large_Plane_Athena_Chrome.NS_Impact_Large_Plane_Athena_Chrome"))

(event EventOnRemoval(EitherVFX-preferthisversion)
  (params (MyTarget :type Actor%20Object%20Reference) (Parameters :type Gameplay%20Cue%20Parameters%20Structure%20%28by%20ref%29) (ParticleComponents :type Array%20of%20FXSystem%20Component%20Object%20References) (AudioComponents :type Array%20of%20Audio%20Component%20Object%20References) (BurstCameraShakeInstance :type Camera%20Shake%20Base%20Object%20Reference) (BurstDecalInstance :type Decal%20Actor%20Object%20Reference)))

(fn ConstructionScript)
