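; Looping gameplay cue notify that gives a Valet basic car Chrome versions of its vehicle effects.
; On loop start it casts the target to Valet_BasicCar_Vehicle and overwrites its FX templates for boost, tire skid, road/exhaust, large impact, tire pop and collision hits.
; Skid and exhaust systems are set on both the Blueprint variables and the native FortDagwoodVehicle ones.
; The removal event is empty, so nothing here restores the original effects.
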(event EventOnLoopingStart(EitherVFX-preferthisversion)
  (params (MyTarget :type Actor%20Object%20Reference) (Parameters :type Gameplay%20Cue%20Parameters%20Structure%20%28by%20ref%29) (ParticleComponents :type Array%20of%20FXSystem%20Component%20Object%20References) (AudioComponents :type Array%20of%20Audio%20Component%20Object%20References))
  (bind _asvalet_basic_car_vehicle (Utilities|Casting|CastToValet_BasicCar_Vehicle MyTarget))
  (Variables|/Valet/BasicCar/Valet_BasicCar_Vehicle.Valet_BasicCar_Vehicle_C|SetTemplateBoostFX "/ChromeShared/FX/NS_Valet_TurboBoost_Chrome.NS_Valet_TurboBoost_Chrome" _asvalet_basic_car_vehicle)
  (Variables|/Valet/BasicCar/Valet_BasicCar_Vehicle.Valet_BasicCar_Vehicle_C|SetFX%20Tire%20Skid%20Template "/ChromeShared/FX/NS_Tire_Skid_Trails_Chrome.NS_Tire_Skid_Trails_Chrome" _asvalet_basic_car_vehicle)
  (Variables|/Valet/BasicCar/Valet_BasicCar_Vehicle.Valet_BasicCar_Vehicle_C|SetFX%20Road%20Interaction%20and%20Exhaust%20Template "/ChromeShared/FX/NS_Wheel_And_Tire_Interaction_Chrome.NS_Wheel_And_Tire_Interaction_Chrome" _asvalet_basic_car_vehicle)
  (Variables|/Script/ValetRuntime.FortDagwoodVehicle|SetNativeSys_TireSkid "/ChromeShared/FX/NS_Tire_Skid_Trails_Chrome.NS_Tire_Skid_Trails_Chrome" _asvalet_basic_car_vehicle)
  (Variables|/Script/ValetRuntime.FortDagwoodVehicle|SetNativeSys_TireExhaust "/ChromeShared/FX/NS_Wheel_And_Tire_Interaction_Chrome.NS_Wheel_And_Tire_Interaction_Chrome" _asvalet_basic_car_vehicle)
  (Variables|/Valet/BasicCar/Valet_BasicCar_Vehicle.Valet_BasicCar_Vehicle_C|SetFX%20Large%20Impact "/ChromeShared/FX/NS_Impact_Large_Plane_Athena_Chrome.NS_Impact_Large_Plane_Athena_Chrome" _asvalet_basic_car_vehicle)
  (Variables|/Valet/BasicCar/Valet_BasicCar_Vehicle.Valet_BasicCar_Vehicle_C|SetPopTireFXNiagara "/ChromeShared/FX/NS_Valet_Tire_Pop_Chrome.NS_Valet_Tire_Pop_Chrome" _asvalet_basic_car_vehicle)
  (Class|ValetBasicCarVehicle|OverrideCollisionHitFX _asvalet_basic_car_vehicle "/ChromeShared/FX/NS_Scrape_PlayerImpact_Chrome.NS_Scrape_PlayerImpact_Chrome" "/ChromeShared/FX/NS_Scrape_Impact_Chrome.NS_Scrape_Impact_Chrome"))

(event EventOnRemoval(EitherVFX-preferthisversion)
  (params (MyTarget :type Actor%20Object%20Reference) (Parameters :type Gameplay%20Cue%20Parameters%20Structure%20%28by%20ref%29) (ParticleComponents :type Array%20of%20FXSystem%20Component%20Object%20References) (AudioComponents :type Array%20of%20Audio%20Component%20Object%20References) (BurstCameraShakeInstance :type Camera%20Shake%20Base%20Object%20Reference) (BurstDecalInstance :type Decal%20Actor%20Object%20Reference)))

(fn ConstructionScript)
